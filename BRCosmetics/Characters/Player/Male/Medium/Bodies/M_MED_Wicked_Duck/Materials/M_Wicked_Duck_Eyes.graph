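# Skeletal mesh material for the eyes of the Wicked Duck player character cosmetic, using Default Lit shading and mesh deformer support.
# It works through material attributes, feeding a Set Material Attributes node into the MF_Wicked_Duck_Eyes function, which holds the eye-specific logic.
# That function's result passes through MF_TechArtOverrides as the final step.
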Material /BRCosmetics/Characters/Player/Male/Medium/Bodies/M_MED_Wicked_Duck/Materials/M_Wicked_Duck_Eyes
graph {
 properties {
  bUseMaterialAttributes=true
  bUsedWithMeshDeformer=true
  bUsedWithSkeletalMesh=true
  pixelDepthOffsetMode=PDOM_Legacy
  usedShadingModels=Default Lit
 }
 nodes [MaterialFunctionCall(In=MaterialFunctionCall!(MaterialAttributes=SetMaterialAttributes!(MaterialAttributes)) {materialFunction=&/MF_Wicked_Duck_Eyes}.Result) {materialFunction=&MF_TechArtOverrides}]
}
