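# Head material for the Wicked Duck skinned character, adding a flickering, blink-aware glow for its cyborg eye.
# Shared PreFX and CharacterSkinFX functions build material attributes, whose emissive is rescaled before passing through PostFX and cinematic post functions.
# A time-driven sine oscillates emissive between minimum and maximum parameters.
# The right-eye blink pose parameter, thresholded by rounding and clamping, switches that glow to a scaled blink intensity.
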Material /BRCosmetics/Characters/Player/Male/Medium/Bodies/M_MED_Wicked_Duck/Materials/M_Wicked_Duck_Head
graph {
 properties {
  bUseMaterialAttributes=true
  bUsedWithMeshDeformer=true
  bUsedWithSkeletalMesh=true
  pixelDepthOffsetMode=PDOM_Legacy
  refractionMethod=RM_IndexOfRefraction
  shadingModel=MSM_FromMaterialExpression
  subsurfaceProfile=&SS_HeroSkin_02
  usedShadingModels=Default Lit | Subsurface | Clear Coat | Subsurface Profile | Cloth | Eye
 }
 nodes [
  n1 = MaterialFunctionCall() {materialFunction=&BaseFN_CharacterFunction_PreFX}
  n2 = GetMaterialAttributes(None=MaterialFunctionCall(Gmap=@n1.Gmap, MaterialAttributes=@n1.Gmap, AO=@n1.Gmap) {materialFunction=&CharacterSkinFX}.Result) {attributeGetTypes [#1]}
  n3 = Multiply(A=@n2.Emissive Color)
  n4 = Multiply(A=Time(), B=ScalarParameter() {defaultValue=5, group=WickedDuckCyborgEye, parameterName=EyeFlickerSpeed, sortPriority=150}) {constB=5}
  n5 = LinearInterpolate(
   A=ScalarParameter() {defaultValue=0.5, group=WickedDuckCyborgEye, parameterName=EyeFlickedMin}
   B=ScalarParameter() {defaultValue=1, group=WickedDuckCyborgEye, parameterName=EyeFlickedMax, sortPriority=100}
   Alpha=ConstantBiasScale(None=Sine(None=@n4))
  ) {constA=0.8}
  n6 = Multiply(
   A=ScalarParameter() {group=Banana_Poses, parameterName=R_blink_pose}
   B=Divide(B=ScalarParameter() {defaultValue=0.3, group=Banana, parameterName=Threshold_R_blink_pose}) {constA=0.5}
  )
  n7 = LinearInterpolate(A=@n5, B=Multiply(A=ScalarParameter() {defaultValue=0.2, parameterName=EyeBlinkIntensity}, B=@n5), Alpha=Clamp(None=Round(None=@n6)))
  n8 = MaterialFunctionCall(
   AO=@n1.AO
   Distance Blend=@n1.AO
   Fresnel=@n1.AO
   Gmap=@n1.AO
   Pre Skinned Local Position=@n1.AO
   Material=SetMaterialAttributes(MaterialAttributes=@n2.MaterialAttributes, Emissive Color=Multiply(A=@n3, B=@n7)) {attributeSetTypes [#1]}
  ) {materialFunction=&BaseFN_CharacterFunction_PostFX}
  MaterialFunctionCall(In=@n8.Result) {materialFunction=&CINE_MF_CharPostFX}
 ]
}
