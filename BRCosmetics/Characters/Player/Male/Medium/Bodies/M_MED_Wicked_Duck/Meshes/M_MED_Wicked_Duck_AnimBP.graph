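# Post-copy AnimBP for the Wicked Duck medium male body cosmetic, adjusting kneepads and shoulder pads atop the parent mesh's pose.
# Weighted bone constraints drive both shoulder pads always and kneepads only while crouching.
# Right upper-arm yaw sets Aim_Up, which rolls the right shoulder pad, except during skydiving or parachuting.
# Each update also checks for long backbling and, if found, toggles two material "Hide Element" parameters.
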AnimBlueprint /BRCosmetics/Characters/Player/Male/Medium/Bodies/M_MED_Wicked_Duck/Meshes/M_MED_Wicked_Duck_AnimBP
target_skeleton=&/M_MED_Wicked_Duck_Skeleton
graphs [
 AnimGraph {nodes [
   "Transform (Modify) Bone - Bone: dyn_l_kneepad_master"(Alpha=1, Translation="0,-0.5,0", Rotation="0,0,0", Scale="1,1,1") {
    rotationMode=BMM_Additive
    translationMode=BMM_Additive
    alphaBoolBlend {blendOption=Linear}
    boneToModify {boneName=dyn_l_kneepad_master}
    translation {y=-0.5}
   }
   n1 = "Constraint - dyn_l_kneepad (PP)"!(
    ComponentPose=Local To Component(LocalPose=Copy Pose From Mesh() {bUseAttachedParent=true})
    Alpha=1
    ConstraintWeights_0=0
    ConstraintWeights_1:float=1
   ) -> (Pose) {
    alphaBoolBlend {blendOption=Linear}
    boneToModify {boneName=dyn_l_kneepad}
    constraintSetup [{targetBone {boneName=thigh_l}}, {targetBone {boneName=calf_l}}]
    constraintWeights [1,1]
   }
   n2 = "Constraint - dyn_r_kneepad (PP)"!(ComponentPose=@n1, Alpha=1, ConstraintWeights_0=0.25, ConstraintWeights_1:float=0.75) -> (Pose) {
    alphaBoolBlend {blendOption=Linear}
    boneToModify {boneName=dyn_r_kneepad}
    constraintSetup [{targetBone {boneName=thigh_r}}, {targetBone {boneName=calf_r}}]
    constraintWeights [1,1]
   }
   n3 = Component To Local(ComponentPose=@n2)
   n4 = Copy Pose From Mesh() {bUseAttachedParent=true}
   n5 = Get bIsCrouching()
   n6 = "Constraint - dyn_shoulderpad_l (PPPPPPPP)"!(
    ComponentPose=Local To Component(LocalPose=Blend Poses by bool(BlendPose_0=@n3, BlendPose_1=@n4, BlendTime_1=0.3, bActiveValue=@n5) {blendType=Linear})
    Alpha=1
    ConstraintWeights_0=0.0924
    ConstraintWeights_1:float=0.0025
    ConstraintWeights_2:float=0.007
    ConstraintWeights_3:float=0.0015
    ConstraintWeights_4:float=0.0094
    ConstraintWeights_5:float=0.8847
    ConstraintWeights_6:float=0.0002
    ConstraintWeights_7:float=0.0023
   ) -> (Pose) {
    alphaBoolBlend {blendOption=Linear}
    boneToModify {boneName=dyn_shoulderpad_l}
    constraintSetup [
     {targetBone {boneName=clavicle_l}}
     {targetBone {boneName=upperarm_twist_02_l}}
     {targetBone {boneName=spine_03}}
     {targetBone {boneName=spine_04}}
     {targetBone {boneName=spine_05}}
     {targetBone {boneName=upperarm_twist_01_l}}
     {targetBone {boneName=neck_02}}
     {targetBone {boneName=upperarm_l}}
    ]
    constraintWeights [1,1,1,1,1,1,1,1]
   }
   n7 = "Constraint - dyn_shoulderpad_r (PPPPPPPP)"!(
    ComponentPose=@n6
    Alpha=1
    ConstraintWeights_0=0.0036
    ConstraintWeights_1:float=0.8512
    ConstraintWeights_2:float=0.004
    ConstraintWeights_3:float=0.0089
    ConstraintWeights_4:float=0.002
    ConstraintWeights_5:float=0.0119
    ConstraintWeights_6:float=0.118
    ConstraintWeights_7:float=0.0002
   ) -> (Pose) {
    alphaBoolBlend {blendOption=Linear}
    boneToModify {boneName=dyn_shoulderpad_r}
    constraintSetup [
     {targetBone {boneName=upperarm_r}}
     {targetBone {boneName=upperarm_twist_01_r}}
     {targetBone {boneName=upperarm_twist_02_r}}
     {targetBone {boneName=spine_03}}
     {targetBone {boneName=spine_04}}
     {targetBone {boneName=spine_05}}
     {targetBone {boneName=clavicle_r}}
     {targetBone {boneName=neck_02}}
    ]
    constraintWeights [1,1,1,1,1,1,1,1]
   }
   n8 = "Transform (Modify) Bone - Bone: dyn_shoulderpad_l"(ComponentPose=@n7, Alpha=1, Translation="0,0,0", Rotation="0,0,0", Scale="1,1,1") {
    rotationMode=BMM_Additive
    scaleMode=BMM_Additive
    translationMode=BMM_Additive
    alphaBoolBlend {blendOption=Linear}
    boneToModify {boneName=dyn_shoulderpad_l}
   }
   n9 = "Transform (Modify) Bone - Bone: dyn_shoulderpad_r_master"(ComponentPose=@n8, Alpha=Get Aim_Up() -> (Aim_Up:double), Translation="0,0,0", Rotation="0,0,-45", Scale="1,1,1") {
    rotationMode=BMM_Additive
    rotationSpace=BCS_ParentBoneSpace
    scaleMode=BMM_Additive
    translationMode=BMM_Additive
    alphaBoolBlend {blendOption=Linear}
    boneToModify {boneName=dyn_shoulderpad_r_master}
    rotation {roll=-45}
   }
   Output Pose(Result=Component To Local(ComponentPose=@n9)) {layerGroup=None}
  ]}
 EventGraph {nodes [
   Event BlueprintUpdateAnimation()
   > n1 = Sequence()
   n2 = Get bIsSkyDiving()
   n3 = Get bIsParachuteOpened()
   > n4 = Branch(Condition=NOT Boolean(A=OR Boolean(A=@n2, B=@n3)))
   n5 = BreakTransform(InTransform=GetDeltaTransformFromRefPose(self=GetOwnerSkeletalMeshComponent(), BoneName=upperarm_r, BaseName=clavicle_r))
   > Set Aim_Up(Aim_Up:double=MapRangeClamped(Value=BreakRotator(InRot=@n5.Rotation).Yaw, InRangeA=60, InRangeB=120)) -> (Output_Get:double)
   Set Aim_Up(execute=@n4.else, Aim_Up=0)
   n6 = CheckForBackbling(LongBackblingThreshold=40)
   Set isWearingBackbling(execute=@n1.then_1, isWearingBackbling=@n6)
   > Set isWearingCape(isWearingCape=@n6.bIsBackblingLong)
   n7 = Get isWearingBackbling()
   n8 = Get isWearingCape() -> (isWearingCape:bool)
   > Branch(Condition=OR Boolean(A=Get isWearingCape(), B=AND Boolean(A=@n7, B=@n8)))
   > n9 = Get Dynamic Material Instance(SkeletalMeshComponent=GetPartSkeletalMeshComponent())
   > SetScalarParameterValue(self=@n9, ParameterName=Hide Element 02, Value=1)
   > SetScalarParameterValue(self=@n9, ParameterName=Hide Element 01, Value=0)
   Event OnCharacterCustomizationLoaded()
  ]}
]
variables [
 Aim_Up {type=number}
 isWearingBackbling {type=boolean}
 isWearingCape {type=boolean}
]
