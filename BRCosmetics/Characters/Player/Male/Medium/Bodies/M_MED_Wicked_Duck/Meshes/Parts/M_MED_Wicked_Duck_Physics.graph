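# Physics asset for the M_MED_Wicked_Duck player body, providing physics for its dangling hair, beak ring and wire parts around the head.
# A head capsule anchors four hair chains (front, midfront, midback, back), a three-link neck wire, a beak ring and left and right head wires.
# Capsules, plus boxes for the head wires, are linked by angle-limited constraints, with hair limits mostly loosening from root toward tip.
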PhysicsAsset /BRCosmetics/Characters/Player/Male/Medium/Bodies/M_MED_Wicked_Duck/Meshes/Parts/M_MED_Wicked_Duck_Physics
bodies {
 dyn_beak_ring [{shapeType=Capsule, length=1.9798, radius=1.9101, center=<2.5183 0.0001 0>}]
 dyn_hair_back_01 [{shapeType=Capsule, length=2, radius=2, center=<2.3456 0 0>, rotation=<rot 90 0 0>}]
 dyn_hair_back_02 [{shapeType=Capsule, length=2, radius=2, center=<2.3456 0 0>, rotation=<rot 90 0 0>}]
 dyn_hair_back_03 [{shapeType=Capsule, length=2, radius=2, center=<2.3456 0 0>, rotation=<rot 90 0 0>}]
 dyn_hair_front_01 [{shapeType=Capsule, length=1.5456, radius=2.4544, center=<2.5164 0.0001 0>, rotation=<rot 90 0 0>}]
 dyn_hair_front_02 [{shapeType=Capsule, length=1.3192, radius=2.6808, center=<2.8566 0 0>, rotation=<rot 90 0 0>}]
 dyn_hair_front_03 [{shapeType=Capsule, length=1.1677, radius=2.8323, center=<3.1232 0.0001 0>, rotation=<rot 90 0 0>}]
 dyn_hair_front_04 [{shapeType=Capsule, length=1.0962, radius=3.5343, center=<4.2767 -0.0001 0>, rotation=<rot 90 0 0>}]
 dyn_hair_front_05 [{shapeType=Capsule, length=1.6352, radius=3.3672, center=<3.8515 -0.0001 0>, rotation=<rot 90 0 0>}]
 dyn_hair_midback_01 [{shapeType=Capsule, length=2, radius=2, center=<2.4387 0.0002 0>, rotation=<rot 90 0 0>}]
 dyn_hair_midback_02 [{shapeType=Capsule, length=2, radius=2, center=<2.4387 0.0002 0>, rotation=<rot 90 0 0>}]
 dyn_hair_midback_03 [{shapeType=Capsule, length=2, radius=2, center=<2.4387 0.0002 0>, rotation=<rot 90 0 0>}]
 dyn_hair_midfront_01 [{shapeType=Capsule, length=2, radius=2, center=<2.6559 0.0001 0>, rotation=<rot 90 0 0>}]
 dyn_hair_midfront_02 [{shapeType=Capsule, length=1.4233, radius=2.5767, center=<3.1375 0.0002 0>, rotation=<rot 90 0 0>}]
 dyn_hair_midfront_03 [{shapeType=Capsule, length=1.1119, radius=2.8881, center=<3.2225 0.0002 0>, rotation=<rot 90 0 0>}]
 dyn_hair_midfront_04 [{shapeType=Capsule, length=0.5003, radius=3.4997, center=<2.6559 0.0001 0>, rotation=<rot 90 0 0>}]
 dyn_hair_midfront_05 [{shapeType=Capsule, length=0.9652, radius=3.0348, center=<3.3607 0.0001 0.0001>, rotation=<rot 90 0 0>}]
 dyn_head_wire_l [{shapeType=Box, extentX=2.7924, extentY=3.6978, extentZ=1.8675, center=<1.6599 0 0>}]
 dyn_head_wire_r [{shapeType=Box, extentX=4.0079, extentY=5.0152, extentZ=1.3345, center=<2.1575 0 0>}]
 dyn_neck_wire_01 [{shapeType=Capsule, length=3.1693, radius=0.6403, center=<2.2068 0 -0.0003>, rotation=<rot 90 0 0>}]
 dyn_neck_wire_02 [{shapeType=Capsule, length=3.1693, radius=0.6403, center=<2.2068 0 -0.0003>, rotation=<rot 90 0 0>}]
 dyn_neck_wire_03 [{shapeType=Capsule, length=1, radius=1, center=<1.0968 0 0>, rotation=<rot 90 0 0>}]
 head [{shapeType=Capsule, length=7.379, radius=13.4587, center=<2.452 1.6994 0>, rotation=<rot 90 0 0>}]
}
constraints [
 {bone1Name=dyn_hair_front_05, bone2Name=dyn_hair_front_04, swing1LimitDegrees=20, swing2LimitDegrees=20, twistLimitDegrees=10}
 {bone1Name=dyn_hair_front_02, bone2Name=dyn_hair_front_01, swing1LimitDegrees=7, swing2LimitDegrees=7, twistLimitDegrees=7}
 {bone1Name=dyn_hair_front_03, bone2Name=dyn_hair_front_02, swing1LimitDegrees=10, swing2LimitDegrees=10, twistLimitDegrees=10}
 {bone1Name=dyn_hair_front_04, bone2Name=dyn_hair_front_03, swing1LimitDegrees=20, swing2LimitDegrees=20, twistLimitDegrees=10}
 {bone1Name=dyn_hair_back_01, bone2Name=head, swing1LimitDegrees=5, swing2LimitDegrees=5, twistLimitDegrees=5}
 {bone1Name=dyn_hair_back_02, bone2Name=dyn_hair_back_01, swing1LimitDegrees=7, swing2LimitDegrees=7, twistLimitDegrees=7}
 {bone1Name=dyn_hair_back_03, bone2Name=dyn_hair_back_02, swing1LimitDegrees=10, swing2LimitDegrees=10, twistLimitDegrees=10}
 {bone1Name=dyn_hair_midback_01, bone2Name=head, swing1LimitDegrees=5, swing2LimitDegrees=5, twistLimitDegrees=5}
 {bone1Name=dyn_hair_midback_02, bone2Name=dyn_hair_midback_01, swing1LimitDegrees=7, swing2LimitDegrees=7, twistLimitDegrees=7}
 {bone1Name=dyn_hair_midback_03, bone2Name=dyn_hair_midback_02, swing1LimitDegrees=10, swing2LimitDegrees=10, twistLimitDegrees=10}
 {bone1Name=dyn_hair_midfront_01, bone2Name=head, swing1LimitDegrees=5, swing2LimitDegrees=5, twistLimitDegrees=5}
 {bone1Name=dyn_hair_midfront_02, bone2Name=dyn_hair_midfront_01, swing1LimitDegrees=7, swing2LimitDegrees=7, twistLimitDegrees=7}
 {bone1Name=dyn_hair_midfront_03, bone2Name=dyn_hair_midfront_02, swing1LimitDegrees=10, swing2LimitDegrees=10, twistLimitDegrees=10}
 {bone1Name=dyn_hair_midfront_04, bone2Name=dyn_hair_midfront_03, swing1LimitDegrees=10, swing2LimitDegrees=10, twistLimitDegrees=10}
 {bone1Name=dyn_hair_midfront_05, bone2Name=dyn_hair_midfront_04, swing1LimitDegrees=10, swing2LimitDegrees=10, twistLimitDegrees=10}
 {bone1Name=dyn_hair_front_01, bone2Name=head, swing1LimitDegrees=5, swing2LimitDegrees=5, twistLimitDegrees=5}
 {bone1Name=dyn_beak_ring, bone2Name=head, swing1LimitDegrees=18, swing2LimitDegrees=5, twistLimitDegrees=5}
 {bone1Name=dyn_head_wire_r, bone2Name=head, swing1LimitDegrees=20, swing2LimitDegrees=20, twistLimitDegrees=10}
 {bone1Name=dyn_head_wire_l, bone2Name=head, swing1LimitDegrees=10, swing2LimitDegrees=25, twistLimitDegrees=10}
 {bone1Name=dyn_neck_wire_03, bone2Name=dyn_neck_wire_02, swing1LimitDegrees=10, swing2LimitDegrees=10, twistLimitDegrees=10}
 {bone1Name=dyn_neck_wire_01, bone2Name=head, swing1LimitDegrees=20, swing2LimitDegrees=20, twistLimitDegrees=10}
 {bone1Name=dyn_neck_wire_02, bone2Name=dyn_neck_wire_01, swing1LimitDegrees=10, swing2LimitDegrees=10, twistLimitDegrees=10}
]
